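# Niagara system for the SodaPop emote, surrounding the player with warm golden glow, dust, sparkles, trails and stage-style spotlights.
# BodyGlow, PixieDust, Sparkles and anim trails sample skeletal mesh bones, with User.FXActive scaling dust, sparkles and trails.
# GroundSparkles trace to slopes and gate spawning on owner velocity; five pulsing, rotating spotlight beams add decal, glow and lights.
# Front-end, platform, quality and time-of-day switches tune color, brightness and offsets.
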NiagaraSystem /BRCosmetics/Animation/Game/MainPlayer/Emotes/SodaPop/VFX/NS_Emote_SodaPop
dependencies {
 usedDataInterfaces [
  &NiagaraDataInterfacePlatformSet
  &NiagaraDataInterfaceArrayFloat3
  &NiagaraDataInterfaceCurve
  &NiagaraDataInterfaceVectorCurve
  &NiagaraDataInterfaceCollisionQuery
  &NiagaraDataInterfaceDebugDraw
  &NiagaraDataInterfaceColorCurve
  &NiagaraDataInterfaceArrayFloat
  &NiagaraDataInterfaceSkeletalMesh
 ]
 usedDynamicInputs [&IsPlatformSetActive, &Boolean_NotOperation]
 usedModules [
  &SystemState
  &self:SystemSpawnScript.SystemScriptSource.SystemScriptGraph.NiagaraNodeAssignment_3.SetVariables_#1
  &self:SystemSpawnScript.SystemScriptSource.SystemScriptGraph.NiagaraNodeAssignment_1.SetVariables_#2
  &self:SystemSpawnScript.SystemScriptSource.SystemScriptGraph.NiagaraNodeAssignment_0.SetVariables_#3
  &self:SystemSpawnScript.SystemScriptSource.SystemScriptGraph.NiagaraNodeAssignment_2.SetVariables_#4
  &self:SystemSpawnScript.SystemScriptSource.SystemScriptGraph.NiagaraNodeAssignment_4.SetVariables_#5
  &self:SystemSpawnScript.SystemScriptSource.SystemScriptGraph.NiagaraNodeAssignment_5.SetVariables_#6
  &self:SystemSpawnScript.SystemScriptSource.SystemScriptGraph.NiagaraNodeAssignment_6.SetVariables_#7
 ]
}
emitters {
 BodyGlow {
  emitterUpdateScript [
   /SetVariables(Emitter.GlowOutlineThickness:NiagaraFloat=&MakeCustomFloatFromBool(Bool=$User.bIsFrontEndPreview, True Float=3, False Float=2.5)) {isSetParameters=true}
   &EmitterState()
   &SpawnBurst_Instantaneous(Spawn Count=7, Spawn Time=0.2, Use Spawn Probability=false, UseLoopCountLimit=false)
  ]
  particleSpawnScript [
   &InitializeParticle(Lifetime=13.267, Mesh Scale Mode=::NewEnumerator4, UsePositionOffset=true, Write Material Random=true)
   &PartitionParticles(Number Of Partitions=7)
   /SetVariables() {enabled=false, isSetParameters=true}
   /SetVariables(Particles.VisibilityTag=&MakeIntFromFloat(Float=&FN_FrontEndFloatModifier(Gameplay=0, Front End=1))) {isSetParameters=true}
  ]
  particleUpdateScript [
   &ParticleState()
   &ScaleMeshSize(Scale Factor=&VectorFromFloat(Value=0.8))
   &ScaleColor(
    Scale Alpha=&MakeCustomFloatFromBool(Bool=$User.bIsFrontEndPreview, True Float=3, False Float=3)
    ScaleA=true
    ScaleRGB=true
   ) {module=ScaleColor001}
   &SkeletalMeshLocation(
    Bone / Socket Index=$Particles.PartitionParticles.Partition
    Bone Sampling Mode=::NewEnumerator4
    MaskSpawnGroup=false
    Orientation Sampling=::NewEnumerator1
    Skeletal Mesh=$System.Skeletal Mesh
   )
   &CameraOffset(Camera Offset Amount=20)
   &DynamicMaterialParameters(Module.Index0Float0Write=true, Module.Index0Float1Write=true, Module.Index0Float2Write=true, Module.Index0Float3Write=true)
  ]
  rendererClasses [&NiagaraMeshRendererProperties]
  renderers [
   {rendererClass=&NiagaraMeshRendererProperties, rendererIndex=0}
   {rendererClass=&NiagaraMeshRendererProperties, rendererIndex=1}
  ]
 }
 GroundSparkles {
  emitterUpdateScript [
   &EmitterState(Life Cycle Mode=::NewEnumerator1, Loop Duration=2, UseLoopDelay=false)
   &SpawnRate(
    SpawnRate=&MakeCustomFloatFromBool(
     Bool=&SetBoolByFloatComparison(A=&VectorLength(Vector=$User.OwnerLocalVelocity), B=100)
     True Float=0
     False Float=30
    )
    Use Spawn Probability=false
   )
  ]
  particleSpawnScript [
   &InitializeParticle(
    Color=&MakeCustomLinearColorFromBool(
     Bool=&SetBoolByFloatComparison(
      A=&FN_PickFloatDependOnPlatform(
       TargetPlatform=DataInterface() {propertyValues {Platforms {
          qualityLevelMask=-29
          deviceProfileStates [
           {profileName=Mobile, qualityLevelMask=4294967295, setQualityLevelMask=4294967295}
           {profileName=Switch, qualityLevelMask=4294967295, setQualityLevelMask=4294967295}
           {profileName=IOS, qualityLevelMask=4294967295, setQualityLevelMask=4294967295}
           {profileName=Android, qualityLevelMask=4294967295, setQualityLevelMask=4294967295}
           {profileName=Mac, qualityLevelMask=0, setQualityLevelMask=3}
           {profileName=Linux, qualityLevelMask=0, setQualityLevelMask=3}
           {profileName=Windows, qualityLevelMask=1, setQualityLevelMask=3}
          ]
         }}}
       TargetPlatformActive=1
       TargetPlatformNoActive=0
      )
      B=1
      Comparison Type=::NewEnumerator4
     )
     True Color=&RandomRangeLinearColor(Minimum=<rgba 1.25 0.7378 0.1197 1>, Maximum=<rgba 1.5 0.9531 0.1259 1>)
     False Color=&RandomRangeLinearColor(Minimum=<rgba 3 1.126 0.3995 1.1>, Maximum=<rgba 3 1.8348 0.379 1.1>)
    )
    Lifetime=&FN_FrontEndFloatModifier(
     Gameplay=&RandomRangeFloat(Minimum=1, Maximum=1.6)
     Front End=&RandomRangeFloat(Minimum=1.6, Maximum=2)
    )
    Sprite Rotation Mode=::NewEnumerator2
    Sprite Size Mode=::NewEnumerator3
    Uniform Sprite Size=&Multiply_Float(A=&RandomRangeFloat(Minimum=125, Maximum=70), B=2)
    UsePositionOffset=false
    Write Material Random=true
   )
   /SetVariables(Particles.Pixie Dust Facing:Vector3f=<0 0 1>) {isSetParameters=true}
   &ShapeLocation(
    Cylinder Height=5
    Cylinder Radius=70
    MaskSpawnGroup=false
    Offset=&RandomRangeVector(Minimum=<0 0 3>, Maximum=<0 0 8>)
    Offset Mode=::NewEnumerator0
    Override Seed=false
    Shape Primitive=::NewEnumerator1
    Surface Only=false
    UseOwnerScale=true
   )
  ]
  particleUpdateScript [
   &ParticleState()
   &ScaleSpriteSize(
    Uniform Curve Scale=0.5
    Uniform Curve Sprite Scale=DataInterface() {propertyValues {Curve {
       defaultValue=0.6
       keys [{interpMode=RCIM_Cubic, tangentMode=RCTM_User, time=0, value=0.35}, {interpMode=RCIM_Cubic, tangentMode=RCTM_User, time=1, value=0.7}]
      }}}
   )
   &SolveForcesAndVelocity(Clamp Velocity=false, Limit Acceleration=false)
   &ScaleColor(
    Scale Alpha=&FloatFromCurve(
     FloatCurve=DataInterface() {propertyValues {Curve {
        defaultValue=0.006
        keys [
         {interpMode=RCIM_Cubic, tangentMode=RCTM_User, time=0, value=0}
         {arriveTangent=0.5346, interpMode=RCIM_Cubic, leaveTangent=0.5346, time=0.4211, value=1}
         {arriveTangent=-1.7143, interpMode=RCIM_Cubic, leaveTangent=-1.7143, tangentMode=RCTM_User, time=0.7084, value=0.3787}
         {tangentMode=RCTM_User, time=1, value=0}
        ]
       }}}
     Scale Curve=&Multiply_Float(
      A=&FNTimeOfDayScalar(Night Scale=0.6, Morning Scale=0.8, Evening Scale=0.8)
      B=&FN_FrontEndFloatModifier(
       Gameplay=&RandomRangeFloat(Minimum=0.03, Maximum=0.05)
       Front End=&RandomRangeFloat(Minimum=0.01, Maximum=0.03)
      )
     )
    )
    Scale RGB=&FN_FrontEndVector3Modifier(
     Gameplay=&VectorFromCurve(VectorCurve=DataInterface() {propertyValues {
        XCurve {keys [<at 0 1 RCIM_Cubic>,<at 1 1 RCIM_Cubic>]}
        YCurve {keys [<at 0 1 RCIM_Cubic>,<at 1 0.75 RCIM_Cubic>]}
        ZCurve {keys [<at 0 1 RCIM_Cubic>,<at 1 0.75 RCIM_Cubic>]}
       }})
     Front End=&VectorFromCurve(VectorCurve=DataInterface() {propertyValues {
        XCurve {keys [
          <at 0 0.25 RCIM_Cubic>
          {arriveTangentWeight=0.1491, interpMode=RCIM_Cubic, leaveTangentWeight=0.2236, tangentMode=RCTM_User, time=0.4, value=1}
          {arriveTangentWeight=0.2, interpMode=RCIM_Cubic, tangentMode=RCTM_User, time=1, value=1}
         ]}
        YCurve {keys [<at 0.0031 0.25 RCIM_Cubic>, {arriveTangent=0.5016, interpMode=RCIM_Cubic, leaveTangent=0.5016, time=0.4, value=1}, <at 1 0.75 RCIM_Cubic>]}
        ZCurve {keys [<at 0 0.25 RCIM_Cubic>, {arriveTangent=0.5, interpMode=RCIM_Cubic, leaveTangent=0.5, time=0.4, value=1}, <at 1 0.75 RCIM_Cubic>]}
       }})
    )
    ScaleA=true
    ScaleRGB=true
   )
   &DynamicMaterialParameters(
    Index 0 Param 1=&FloatFromCurve(FloatCurve=DataInterface() {propertyValues {Curve {keys [<at 0 0>,<at 0.6294 0.0067>,<at 1 0.02>]}}}, Scale Curve=0)
    Module.Index0Float0Write=false
    Module.Index0Float1Write=true
    Module.Index0Float2Write=true
    Module.Index0Float3Write=true
   )
   /SetVariables(
    Particles.SparkleColor:LinearColor=&MakeLinearColorFromVectorAndFloat(
     "Vector (RGB)"=&MakeVectorFromLinearColorRGB(Color=&RandomRangeLinearColor(Minimum=<rgba 1.5 0.9639 0.0054 1>, Maximum=<rgba 3 2.148 0.4033 1>))
     "Float (Alpha)"=&FloatFromCurve(FloatCurve=DataInterface() {propertyValues {Curve {
         defaultValue=1.0194
         keys [
          {interpMode=RCIM_Cubic, tangentMode=RCTM_User, time=0, value=0}
          {
           arriveTangent=7.0019
           arriveTangentWeight=0.0547
           interpMode=RCIM_Cubic
           leaveTangent=7.0019
           leaveTangentWeight=0.107
           tangentMode=RCTM_User
           time=0.1641
           value=1
          }
          {arriveTangent=0.3424, interpMode=RCIM_Cubic, leaveTangent=0.3424, time=0.3242, value=2}
          {
           arriveTangent=-6.1112
           arriveTangentWeight=0.1926
           interpMode=RCIM_Cubic
           leaveTangent=-6.1112
           leaveTangentWeight=0.1149
           tangentMode=RCTM_User
           time=0.5044
           value=1.1165
          }
          {arriveTangent=-2.2528, interpMode=RCIM_Cubic, leaveTangent=-2.2528, time=0.6763, value=0.2328}
          {tangentMode=RCTM_User, time=1, value=0}
         ]
        }}})
    )
    Particles.Vector 2D:Vector2f=&Multiply_Vector2D_ByFloat(
     Float=&FloatFromCurve(FloatCurve=DataInterface() {propertyValues {Curve {keys [
          <at 0 0 RCIM_Cubic>
          {arriveTangent=0.5233, interpMode=RCIM_Cubic, leaveTangent=0.5233, time=0.2, value=0.95}
          {arriveTangent=-0.3966, interpMode=RCIM_Cubic, leaveTangent=-0.3966, time=0.4754, value=0.2488}
          {arriveTangent=-0.4742, interpMode=RCIM_Cubic, leaveTangent=-0.4742, time=0.7179, value=0.7446}
          <at 1 0 RCIM_Cubic>
         ]}}})
     Vector2D=$Particles.Initial.SpriteSize
    )
   ) {isSetParameters=true}
   &self:Slope_Position_06(
    Apply Position=true
    Apply Rotation=true
    Fallback Position:NiagaraPosition=$Particles.Initial.ShapeLocation.ShapePosition
    Height Adjustment/Offset=1
    Slope Position:NiagaraPosition=$Particles.Initial.ShapeLocation.ShapePosition
    Trace Length=180
    Trace Origin Height=85
    bIsFrontEnd:NiagaraBool=$User.bIsFrontEndPreview
   ) {module=Slope_Position_06}
  ]
  rendererClasses [&NiagaraSpriteRendererProperties]
  renderers [{rendererClass=&NiagaraSpriteRendererProperties}]
 }
 PixieDust {
  emitterUpdateScript [
   &EmitterState()
   &SpawnRate(SpawnRate=&Multiply_Float(A=150, B=$User.FXActive), Use Spawn Probability=false)
   &SpawnRate(SpawnGroup=1, SpawnRate=&Multiply_Float(A=25, B=$User.FXActive), Use Spawn Probability=false) {module=SpawnRate001}
  ]
  particleSpawnScript [
   /SetVariables(Particles.VisibilityTag=&SpawnGroup()) {isSetParameters=true}
   &PartitionParticles()
   &InitializeParticle(
    Color=&MakeCustomLinearColorFromBool(
     Bool=&SetBoolByFloatComparison(
      A=&FN_PickFloatDependOnPlatform(
       TargetPlatform=DataInterface() {propertyValues {Platforms {
          qualityLevelMask=-29
          deviceProfileStates [
           {profileName=Mobile, qualityLevelMask=4294967295, setQualityLevelMask=4294967295}
           {profileName=Switch, qualityLevelMask=4294967295, setQualityLevelMask=4294967295}
           {profileName=IOS, qualityLevelMask=4294967295, setQualityLevelMask=4294967295}
           {profileName=Android, qualityLevelMask=4294967295, setQualityLevelMask=4294967295}
           {profileName=Mac, qualityLevelMask=0, setQualityLevelMask=3}
           {profileName=Linux, qualityLevelMask=0, setQualityLevelMask=3}
           {profileName=Windows, qualityLevelMask=1, setQualityLevelMask=3}
          ]
         }}}
       TargetPlatformActive=1
       TargetPlatformNoActive=0
      )
      B=1
      Comparison Type=::NewEnumerator4
     )
     True Color=&RandomRangeLinearColor(Minimum=<rgba 1 0.5902 0.0957 0.7078>, Maximum=<rgba 1 0.6354 0.0839 0.4209>)
     False Color=&RandomRangeLinearColor(Minimum=<rgba 1 0.8063 0.3229 1.1>, Maximum=<rgba 1 0.5602 0.3258 1.1>)
    )
    Lifetime=&MaskFloat(
     Mask Source=0
     Mask Parameter=?
     Mask=0
     Masked Float=&RandomRangeFloat(Minimum=0.5, Maximum=1)
     Passthrough Float=&RandomRangeFloat(Minimum=1, Maximum=2)
    )
    Position Offset=<2 0 -2>
    Sprite Rotation Mode=::NewEnumerator2
    Sprite Size Mode=::NewEnumerator3
    Uniform Sprite Size=&MaskFloat(
     Mask Source=0
     Mask Parameter=?
     Mask=0
     Masked Float=&RandomRangeFloat(Minimum=35, Maximum=65)
     Passthrough Float=&RandomRangeFloat(Minimum=2, Maximum=5)
    )
    UsePositionOffset=true
    Write Material Random=true
   )
   &SkeletalMeshLocation(
    Bone / Socket Index=$Particles.PartitionParticles.Partition
    Bone Sampling Mode=::NewEnumerator4
    MaskSpawnGroup=false
    Sampled Position Offset=&SelectVectorFromArray(
     Direct Array Index=$Particles.PartitionParticles.Partition
     Vector Selection Array=DataInterface() {propertyValues {FloatData [<-5 0 0>, <5 0 0>]}}
    )
    Skeletal Mesh=$System.Skeletal Mesh
   ) {module=SkeletalMeshLocation001}
   &AddVelocity(
    Module.Use Falloff Distance=false
    Override Seed=false
    Velocity Mode=::NewEnumerator1
    Velocity Origin=$Output.SkeletalMeshLocation001.SampledPositionWithoutOffset
   )
  ]
  particleUpdateScript [
   &ParticleState()
   &ScaleSpriteSize(Uniform Curve Sprite Scale=DataInterface() {propertyValues {Curve {
       defaultValue=0.6
       keys [{interpMode=RCIM_Cubic, tangentMode=RCTM_User, time=0, value=0}, {interpMode=RCIM_Cubic, tangentMode=RCTM_User, time=1, value=1}]
      }}})
   &InheritSourceMovement(
    Applied Position Scale=&MakeCustomVectorFromBool(
     Bool=&SetBoolByFloatComparison(A=&VectorLength(Vector=$User.OwnerLocalVelocity), B=100)
     True Vector=<0 0 0>
     False Vector=<0.7 0.7 0.7>
    )
    ApplyForce=false
    ApplyPosition=true
    ApplyPositionFalloff=false
   )
   &CurlNoiseForce(
    Bias Noise=false
    Noise Frequency=&RandomRangeFloat(Minimum=10, Maximum=25)
    Noise Strength=&RandomRangeFloat(Minimum=60, Maximum=80)
    Pan Noise=false
   )
   &Drag(Use Linear Drag=true, Use Rotational Drag=true)
   &InheritVelocity(
    Inherited Velocity Amount Scale=&MakeCustomVectorFromBool(
     Bool=&SetBoolByFloatComparison(A=&VectorLength(Vector=$User.OwnerLocalVelocity), B=100)
     True Vector=<0 0 0>
     False Vector=<1 1 -1>
    )
    Inherited Velocity Speed Limit=4
    Speed Limit=true
    UseSpeedThreshold=false
    Velocity Source=$Particles.SkeletalMeshLocation001.SampledVelocity
   )
   &GravityForce(Gravity=<0 0 -120>)
   &SolveForcesAndVelocity(Clamp Velocity=false, Limit Acceleration=false)
   &ScaleColor(
    Scale Alpha=&FloatFromCurve(
     FloatCurve=DataInterface() {propertyValues {Curve {
        defaultValue=0.006
        keys [
         {interpMode=RCIM_Cubic, tangentMode=RCTM_User, time=0, value=0}
         {arriveTangent=0.5346, interpMode=RCIM_Cubic, leaveTangent=0.5346, time=0.4211, value=1}
         {arriveTangent=-1.7143, interpMode=RCIM_Cubic, leaveTangent=-1.7143, tangentMode=RCTM_User, time=0.7084, value=0.3787}
         {tangentMode=RCTM_User, time=1, value=0}
        ]
       }}}
     Scale Curve=&Multiply_Float(
      A=&FNTimeOfDayScalar(Night Scale=0.6, Morning Scale=0.8, Evening Scale=0.8)
      B=&FN_FrontEndFloatModifier(
       Gameplay=&RandomRangeFloat(Minimum=0.25, Maximum=0.35)
       Front End=&RandomRangeFloat(Minimum=0.2, Maximum=0.3)
      )
     )
    )
    Scale RGB=&VectorFromCurve(VectorCurve=DataInterface() {propertyValues {
       XCurve {keys [<at 0 1 RCIM_Cubic>,<at 1 1 RCIM_Cubic>]}
       YCurve {keys [<at 0 1 RCIM_Cubic>,<at 1 0.75 RCIM_Cubic>]}
       ZCurve {keys [<at 0 1 RCIM_Cubic>,<at 1 0.75 RCIM_Cubic>]}
      }})
    ScaleA=true
    ScaleRGB=true
   )
   &DynamicMaterialParameters(
    Index 0 Param 1=&FloatFromCurve(FloatCurve=DataInterface() {propertyValues {Curve {keys [<at 0 0>,<at 1 0.02>]}}}, Scale Curve=0)
    Module.Index0Float0Write=true
    Module.Index0Float1Write=true
    Module.Index0Float2Write=true
    Module.Index0Float3Write=true
   )
   /SetVariables(
    Particles.SparkleColor:LinearColor=&MakeLinearColorFromVectorAndFloat(
     "Vector (RGB)"=&MakeVectorFromLinearColorRGB(Color=&RandomRangeLinearColor(Minimum=<rgba 1.5 0.9639 0.0054 1>, Maximum=<rgba 3 2.148 0.4033 1>))
     "Float (Alpha)"=&FloatFromCurve(FloatCurve=DataInterface() {propertyValues {Curve {
         defaultValue=1.0194
         keys [
          {interpMode=RCIM_Cubic, tangentMode=RCTM_User, time=0, value=0}
          {
           arriveTangent=7.0019
           arriveTangentWeight=0.0547
           interpMode=RCIM_Cubic
           leaveTangent=7.0019
           leaveTangentWeight=0.107
           tangentMode=RCTM_User
           time=0.1641
           value=1
          }
          {arriveTangent=0.3424, interpMode=RCIM_Cubic, leaveTangent=0.3424, time=0.3242, value=2}
          {
           arriveTangent=-6.1112
           arriveTangentWeight=0.1926
           interpMode=RCIM_Cubic
           leaveTangent=-6.1112
           leaveTangentWeight=0.1149
           tangentMode=RCTM_User
           time=0.5044
           value=1.1165
          }
          {arriveTangent=-2.2528, interpMode=RCIM_Cubic, leaveTangent=-2.2528, time=0.6763, value=0.2328}
          {tangentMode=RCTM_User, time=1, value=0}
         ]
        }}})
    )
    Particles.Vector 2D:Vector2f=&Multiply_Vector2D_ByFloat(
     Float=&FloatFromCurve(FloatCurve=DataInterface() {propertyValues {Curve {keys [
          <at 0 0 RCIM_Cubic>
          {arriveTangent=0.5233, interpMode=RCIM_Cubic, leaveTangent=0.5233, time=0.2, value=0.95}
          {arriveTangent=-0.3966, interpMode=RCIM_Cubic, leaveTangent=-0.3966, time=0.4754, value=0.2488}
          {arriveTangent=-0.4742, interpMode=RCIM_Cubic, leaveTangent=-0.4742, time=0.7179, value=0.7446}
          <at 1 0 RCIM_Cubic>
         ]}}})
     Vector2D=$Particles.Initial.SpriteSize
    )
   ) {isSetParameters=true}
  ]
  rendererClasses [&NiagaraSpriteRendererProperties]
  renderers [
   {rendererClass=&NiagaraSpriteRendererProperties}
   {rendererClass=&NiagaraSpriteRendererProperties, rendererIndex=1}
  ]
 }
 Sparkles {
  emitterUpdateScript [
   &EmitterState(Life Cycle Mode=::NewEnumerator1, Loop Delay=0.3, Loop Duration=13.267, UseLoopDelay=false)
   &SpawnRate(SpawnRate=&Multiply_Float(A=70, B=$User.FXActive), Use Spawn Probability=false)
   &SpawnRate(SpawnGroup=1, SpawnRate=&Multiply_Float(A=10, B=$User.FXActive), Use Spawn Probability=false) {module=SpawnRate001}
  ]
  particleSpawnScript [
   /SetVariables(Particles.VisibilityTag=&SpawnGroup()) {isSetParameters=true}
   &PartitionParticles()
   &InitializeParticle(
    Color Maximum=<rgba 1.35 0.9666 0.1815 1>
    Color Minimum=<rgba 1 0.6426 0.0036 1>
    Color Mode=::NewEnumerator2
    Lifetime=&MaskFloat(
     Mask Source=0
     Mask Parameter=?
     Mask=0
     Masked Float=&RandomRangeFloat(Minimum=0.25, Maximum=0.5)
     Passthrough Float=&RandomRangeFloat(Minimum=0.5, Maximum=0.8)
    )
    Sprite Rotation Mode=::NewEnumerator2
    Sprite Size Mode=::NewEnumerator3
    Uniform Sprite Size=&MaskFloat(
     Mask Source=0
     Mask Parameter=?
     Mask=0
     Masked Float=&RandomRangeFloat(Minimum=2.5, Maximum=10)
     Passthrough Float=&RandomRangeFloat(Minimum=3, Maximum=6)
    )
    UsePositionOffset=false
    Write Material Random=true
   )
   &ShapeLocation(MaskSpawnGroup=false, Override Seed=false, Sphere Radius=10, UseOwnerScale=true)
   &AddVelocity(Module.Use Falloff Distance=false, Override Seed=false, Velocity Mode=::NewEnumerator1, Velocity Speed=30)
   &SkeletalMeshLocation(
    Bone / Socket Index=$Particles.PartitionParticles.Partition
    Bone Sampling Mode=::NewEnumerator4
    MaskSpawnGroup=false
    Sampled Position Offset=&SelectVectorFromArray(
     Direct Array Index=$Particles.PartitionParticles.Partition
     Vector Selection Array=DataInterface() {propertyValues {FloatData [<-5 0 0>, <5 0 0>]}}
    )
    Skeletal Mesh=$System.Skeletal Mesh Reader
   ) {module=SkeletalMeshLocation001}
  ]
  particleUpdateScript [
   &ParticleState()
   &ScaleSpriteSize(Uniform Curve Sprite Scale=DataInterface() {propertyValues {Curve {keys [
        <at 0 0 RCIM_Cubic>
        {arriveTangent=0.5233, interpMode=RCIM_Cubic, leaveTangent=0.5233, time=0.2, value=0.95}
        {arriveTangent=-0.3966, interpMode=RCIM_Cubic, leaveTangent=-0.3966, time=0.4754, value=0.2488}
        {arriveTangent=-0.4742, interpMode=RCIM_Cubic, leaveTangent=-0.4742, time=0.7179, value=0.7446}
        <at 1 0 RCIM_Cubic>
       ]}}})
   &ScaleColor(
    Scale Alpha=&FloatFromCurve(FloatCurve=DataInterface() {propertyValues {Curve {
        defaultValue=1.0194
        keys [
         {interpMode=RCIM_Cubic, tangentMode=RCTM_User, time=0, value=0}
         {
          arriveTangent=7.0019
          arriveTangentWeight=0.0547
          interpMode=RCIM_Cubic
          leaveTangent=7.0019
          leaveTangentWeight=0.107
          tangentMode=RCTM_User
          time=0.1641
          value=1
         }
         {arriveTangent=0.3424, interpMode=RCIM_Cubic, leaveTangent=0.3424, time=0.3242, value=2}
         {
          arriveTangent=-6.1112
          arriveTangentWeight=0.1926
          interpMode=RCIM_Cubic
          leaveTangent=-6.1112
          leaveTangentWeight=0.1149
          tangentMode=RCTM_User
          time=0.5044
          value=1.1165
         }
         {arriveTangent=-2.2528, interpMode=RCIM_Cubic, leaveTangent=-2.2528, time=0.6763, value=0.2328}
         {tangentMode=RCTM_User, time=1, value=0}
        ]
       }}})
    ScaleA=true
    ScaleRGB=true
   )
   /SetVariables(Particles.ColorGlow:LinearColor=&MakeLinearColorFromVectorAndFloat(
     "Vector (RGB)"=&MakeVectorFromLinearColorRGB(Color=<rgba 1 0.1468 0 1>)
     "Float (Alpha)"=&FloatFromCurve(
      FloatCurve=DataInterface() {propertyValues {Curve {keys [
          <at 0 0>
          {arriveTangent=1.25, leaveTangent=1.25, time=0.2028, value=0.9955}
          {arriveTangent=-1.2487, leaveTangent=-1.2487, time=0.8, value=1}
          <at 1 0>
         ]}}}
      Scale Curve=0.11
     )
    )) {isSetParameters=true}
   /SetVariables(Particles.SizeGlow:Vector2f=&Multiply_Vector2D_ByFloat(Float=3, Vector2D=$Particles.SpriteSize)) {isSetParameters=true}
   &CurlNoiseForce(
    Bias Noise=false
    Noise Frequency=&RandomRangeFloat(Minimum=10, Maximum=25)
    Noise Strength=&RandomRangeFloat(Minimum=60, Maximum=80)
    Pan Noise=false
   )
   &GravityForce(Gravity=<0 0 -60>)
   &InheritSourceMovement(ApplyForce=false, ApplyPosition=true, ApplyPositionFalloff=false)
   &SolveForcesAndVelocity(Clamp Velocity=false, Limit Acceleration=false)
  ]
  rendererClasses [&NiagaraSpriteRendererProperties]
  renderers [
   {rendererClass=&NiagaraSpriteRendererProperties}
   {rendererClass=&NiagaraSpriteRendererProperties, rendererIndex=1}
   {rendererClass=&NiagaraSpriteRendererProperties, rendererIndex=2}
  ]
 }
 Spotlights {
  emitterUpdateScript [
   &EmitterState(
    Inactive Response=::NewEnumerator1
    Life Cycle Mode=::NewEnumerator1
    Loop Behavior=::NewEnumerator1
    Loop Duration Mode=::NewEnumerator1
    UseLoopDelay=false
   )
   &SpawnBurst_Instantaneous(Spawn Count=5, Use Spawn Probability=false, UseLoopCountLimit=false)
   &SpawnBurst_Instantaneous(Spawn Group=1, Use Spawn Probability=false, UseLoopCountLimit=false) {module=SpawnBurst_Instantaneous001}
  ]
  particleSpawnScript [
   /SetVariables(Particles.VisibilityTag=&SpawnGroup()) {isSetParameters=true}
   &PartitionParticles(Number Of Partitions=5)
   &InitializeParticle(
    Color=<rgba 2 1.3244 0 0.5565>
    Mesh Scale=&MergeHelper_VectorSwitch(
     0=<0.9 0.9 1.5>
     1=<0.9 0.9 1.2>
     2=<0.9 0.9 1.5>
     3=<0.4 0.4 1.2>
     4=<0.4 0.4 1.2>
     Index=$Particles.PartitionParticles.Partition
     Num=::NewEnumerator3
    )
    Mesh Scale Mode=::NewEnumerator4
    Sprite Size Mode=::NewEnumerator3
    Uniform Sprite Size=&MergeHelper_FloatSwitch(0=111, 1=111, 2=111, 3=55.5, 4=55.5, Index=$Particles.PartitionParticles.Partition, Num=::NewEnumerator3)
    UsePositionOffset=false
    Write Material Random=true
   )
   /SetVariables(Particles.Position=&MergeHelper_VectorSwitch(
     0=<100 -100 250>
     1=<0 -50 250>
     2=<-100 -100 250>
     3=<40 -110 210>
     4=<-40 -110 210>
     Index=$Particles.PartitionParticles.Partition
     Num=::NewEnumerator3
    )) {isSetParameters=true}
   &InitialMeshOrientation(
    Mesh Axis To Orient=&MergeHelper_VectorSwitch(
     0=<-0.3 0.5 1>
     1=<0 1 -0.5>
     2=<0.3 0.5 1>
     3=<-0.7 1 -1>
     4=<0.7 1 -1>
     Index=$Particles.PartitionParticles.Partition
     Num=::NewEnumerator3
    )
    Mesh Orientation Mode=::NewEnumerator2
    Orientation Vector=&MergeHelper_VectorSwitch(0=<0 0 1>, 1=<0 1 0>, 2=<0 0 1>, 3=<0 1 0>, 4=<0 1 0>, Index=$Particles.PartitionParticles.Partition, Num=::NewEnumerator3)
    Rotation Coordinate Space=::NewEnumerator0
    Use Rotation Vector=true
   )
   /SetVariables(Particles.WideScale:Vector3f=&Multiply_Vector(A=$Particles.Scale, B=<1.35 1.25 1>)) {isSetParameters=true}
   /SetVariables(Particles.WideColor:LinearColor=<rgba 2 0.5355 0 0.7826>) {isSetParameters=true}
   /SetVariables(
    Particles.Decal Color:LinearColor=&MultiplyLinearColor(
     Linear Color=&MakeCustomLinearColorFromBool(
      Bool=&Boolean_OrOperation(Bool1=$User.bIsFrontEnd, Bool2=$User.bIsFrontEndPreview)
      True Color=<rgba 1.037 0.3757 0 0.15>
      False Color=&FNTimeOfDayColor(
       Night Color=<rgba 1 0.4831 0.1491 0.15>
       Morning Color=<rgba 1 0.5076 0.1011 0.15>
       Day Color=<rgba 1 0.4653 0.1198 0.15>
       Evening Color=<rgba 1 0.4951 0.0787 0.15>
      )
     )
     Scale Factor=&MakeVector4FromVectorAndFloat(
      "Vector (RGB)"=<1 1 1>
      "Float (Alpha)"=&Multiply_Float(
       A=&FNTimeOfDayScalar(Night Scale=0.7, Morning Scale=0.9, Evening Scale=0.9)
       B=&MakeCustomFloatFromBool(
        Bool=&SetBoolByIntComparison(A=$Engine.QualityLevel, B=1, Comparison Type=::NewEnumerator7)
        True Float=0.6
        False Float=1
       )
      )
     )
    )
    Particles.Decal Pos:NiagaraPosition=<-10 -20 0>
   ) {isSetParameters=true}
   /SetVariables(
    Particles.Filler Glow Cam Offset=-30
    Particles.Filler Glow Color:LinearColor=&MakeCustomLinearColorFromBool(
     Bool=&Boolean_OrOperation(Bool1=$User.bIsFrontEnd, Bool2=$User.bIsFrontEndPreview)
     True Color=<rgba 0.2552 0.0597 0 0.4>
     False Color=&FNTimeOfDayColor(
      Night Color=<rgba 0.1146 0.0411 0 0.4>
      Morning Color=<rgba 0.4479 0.1104 0.0155 0.4>
      Day Color=<rgba 0.2552 0.0597 0 0.4>
      Evening Color=<rgba 0.4479 0.1104 0.0155 0.4>
     )
    )
    Particles.Filler Glow Pos:NiagaraPosition=&MakeCustomVectorFromBool(
     Bool=&SetBoolByIntComparison(A=$Engine.QualityLevel, B=1, Comparison Type=::NewEnumerator7)
     True Vector=<0 0 120>
     False Vector=&FN_FrontEndVector3Modifier(Gameplay=<0 0 100>, Front End=<0 0 80>)
    )
    Particles.Filler Glow Size:Vector2f=&MakeCustomVector2DFromBool(
     Bool=&SetBoolByIntComparison(A=$Engine.QualityLevel, B=1, Comparison Type=::NewEnumerator7)
     True Vector2D=<400 370>
     False Vector2D=<500 500>
    )
   ) {isSetParameters=true}
   /SetVariables(
    Particles.Back Light Pos:NiagaraPosition=<0 -75 120>
    Particles.Front Light Pos:NiagaraPosition=<0 75 120>
    Particles.Lights Color:LinearColor=<rgba 5 3.4204 0 1>
   ) {isSetParameters=true}
  ]
  particleUpdateScript [
   &ParticleState(
    Kill Particles When Lifetime Has Elapsed=false
    Let Infinitely Lived Particles Die When Emitter Deactivates=true
    Loop Particles Lifetime=true
   )
   /SetVariables(Particles.Intensity:NiagaraFloat=&Waveform(
     "Waveform [1]"=::NewEnumerator0
     "[1] Frequency"=&MergeHelper_FloatSwitch(0=0.2, 1=0.35, 2=0.25, 3=0.3, 4=0.2, Index=$Particles.PartitionParticles.Partition, Num=::NewEnumerator3)
     "[1] Amplitude Min/Max"=&MakeCustomVector2DFromBool(
      Bool=&SetBoolByFloatComparison(
       A=&FN_PickFloatDependOnPlatform(
        TargetPlatform=DataInterface() {propertyValues {Platforms {
           qualityLevelMask=-29
           deviceProfileStates [
            {profileName=Mobile, qualityLevelMask=4294967295, setQualityLevelMask=4294967295}
            {profileName=Switch, qualityLevelMask=4294967295, setQualityLevelMask=4294967295}
            {profileName=IOS, qualityLevelMask=4294967295, setQualityLevelMask=4294967295}
            {profileName=Android, qualityLevelMask=4294967295, setQualityLevelMask=4294967295}
            {profileName=Mac, qualityLevelMask=0, setQualityLevelMask=3}
            {profileName=Linux, qualityLevelMask=0, setQualityLevelMask=3}
            {profileName=Windows, qualityLevelMask=1, setQualityLevelMask=3}
           ]
          }}}
        TargetPlatformActive=1
        TargetPlatformNoActive=0
       )
       B=1
       Comparison Type=::NewEnumerator4
      )
      True Vector2D=<0.1 1>
      False Vector2D=<0.1 1>
     )
    )) {isSetParameters=true}
   &UpdateMeshOrientation(
    Rotation Rate=&MergeHelper_FloatSwitch(
     0=&RandomRangeFloat(Minimum=0, Maximum=0.25, Evaluation Type=::NewEnumerator1)
     1=&RandomRangeFloat(Minimum=0, Maximum=0.3, Evaluation Type=::NewEnumerator1)
     2=&RandomRangeFloat(Minimum=0, Maximum=0.25, Evaluation Type=::NewEnumerator1)
     3=&RandomRangeFloat(Minimum=0, Maximum=0.35, Evaluation Type=::NewEnumerator1)
     4=&RandomRangeFloat(Minimum=0, Maximum=0.35, Evaluation Type=::NewEnumerator1)
     Index=$Particles.PartitionParticles.Partition
     Num=::NewEnumerator3
    )
    Rotation Vector=&MergeHelper_VectorSwitch(0=<0.2 0 1>, 1=<0 0.1 1>, 2=<-0.2 0 1>, 3=<-0.1 -0.1 1>, 4=<0.1 -0.1 1>, Index=$Particles.PartitionParticles.Partition, Num=::NewEnumerator3)
   ) {module=UpdateMeshOrientation001}
   /SetVariables(
    Particles.GlowCameraOffset:NiagaraFloat=&MakeCustomFloatFromBool(
     Bool=&SetBoolByFloatComparison(
      A=&FN_PickFloatDependOnPlatform(
       TargetPlatform=DataInterface() {propertyValues {Platforms {
          qualityLevelMask=-31
          deviceProfileStates [
           {profileName=Mobile, qualityLevelMask=4294967295, setQualityLevelMask=4294967295}
           {profileName=Switch, qualityLevelMask=4294967295, setQualityLevelMask=4294967295}
           {profileName=IOS, qualityLevelMask=4294967295, setQualityLevelMask=4294967295}
           {profileName=Android, qualityLevelMask=4294967295, setQualityLevelMask=4294967295}
           {profileName=Mac, qualityLevelMask=0, setQualityLevelMask=3}
           {profileName=Linux, qualityLevelMask=0, setQualityLevelMask=3}
           {profileName=Windows, qualityLevelMask=1, setQualityLevelMask=3}
          ]
         }}}
       TargetPlatformActive=1
       TargetPlatformNoActive=0
      )
      B=1
      Comparison Type=::NewEnumerator4
     )
     True Float=-50
     False Float=11
    )
    Particles.GlowColor=&MultiplyLinearColor(Linear Color=$Particles.Color, Scale Factor=&MakeVector4_FromFloat(Float=0.6))
   ) {isSetParameters=true}
   /SetVariables(
    Particles.Cam Offset Low Platforms:NiagaraFloat=&MakeCustomFloatFromBool(
     Bool=&SetBoolByFloatComparison(
      A=&FN_PickFloatDependOnPlatform(
       TargetPlatform=DataInterface() {propertyValues {Platforms {
          qualityLevelMask=-29
          deviceProfileStates [
           {profileName=Mobile, qualityLevelMask=4294967295, setQualityLevelMask=4294967295}
           {profileName=Switch, qualityLevelMask=4294967295, setQualityLevelMask=4294967295}
           {profileName=IOS, qualityLevelMask=4294967295, setQualityLevelMask=4294967295}
           {profileName=Android, qualityLevelMask=4294967295, setQualityLevelMask=4294967295}
           {profileName=Mac, qualityLevelMask=0, setQualityLevelMask=3}
           {profileName=Linux, qualityLevelMask=0, setQualityLevelMask=3}
           {profileName=Windows, qualityLevelMask=1, setQualityLevelMask=3}
          ]
         }}}
       TargetPlatformActive=1
       TargetPlatformNoActive=0
      )
      B=1
      Comparison Type=::NewEnumerator4
     )
     True Float=-100
     False Float=0
    )
    Particles.Update Glow Color:LinearColor=&MultiplyLinearColor(
     Linear Color=$Particles.GlowColor
     Scale Factor=&MakeVector4_FromLinearColor(LinearColor=&MakeLinearColorFromVectorAndFloat(
       "Vector (RGB)"=<1 1 1>
       "Float (Alpha)"=&Multiply_Float(A=&Multiply_Float(A=0.5, B=$Particles.Intensity), B=1)
      ))
    )
    Particles.UpdateNarrowColor:LinearColor=&MultiplyLinearColor(
     Linear Color=$Particles.Initial.Color
     Scale Factor=&MakeVector4_FromLinearColor(LinearColor=&MakeLinearColorFromVectorAndFloat(
       "Vector (RGB)"=<1 1 1>
       "Float (Alpha)"=&Multiply_Float(
        A=&Multiply_Float(A=0.4, B=$Particles.Intensity)
        B=&FN_FrontEndFloatModifier(Gameplay=1, Front End=0.75)
       )
      ))
    )
    Particles.UpdateWideColor:LinearColor=&MultiplyLinearColor(
     Linear Color=$Particles.Initial.WideColor
     Scale Factor=&MakeVector4_FromLinearColor(LinearColor=&MakeLinearColorFromVectorAndFloat(
       "Vector (RGB)"=<1 1 1>
       "Float (Alpha)"=&Multiply_Float(
        A=&Multiply_Float(A=0.75, B=$Particles.Intensity)
        B=&FN_FrontEndFloatModifier(Gameplay=1, Front End=0.75)
       )
      ))
    )
   ) {isSetParameters=true}
   /Niagara/Modules/Decal/Decal_Attributes(Decal Size:Vector3f=<60 1000 1000>, Fade=1, Size=1)
   &Light_Attributes(
    Diffuse Scale=1
    Falloff Exponent=1
    Light Diffuse Scale=1.25
    Light Exponent=0.9
    Light Position Offset:Vector3f=<0 50 100>
    Light Radius=70
    Position=1
    Radius=1
    Specular Scale=1
    UsePositionOffset=true
    Volumetric Scattering=1
   )
  ]
  rendererClasses [
   &NiagaraMeshRendererProperties
   &NiagaraSpriteRendererProperties
   &NiagaraDecalRendererProperties
   &NiagaraLightRendererProperties
  ]
  renderers [
   {rendererClass=&NiagaraMeshRendererProperties, rendererIndex=0}
   {rendererClass=&NiagaraMeshRendererProperties, rendererIndex=1}
   {rendererClass=&NiagaraSpriteRendererProperties, rendererIndex=2}
   {rendererClass=&NiagaraDecalRendererProperties, rendererIndex=3}
   {rendererClass=&NiagaraSpriteRendererProperties, rendererIndex=4}
   {rendererClass=&NiagaraLightRendererProperties, rendererIndex=5}
   {rendererClass=&NiagaraLightRendererProperties, rendererIndex=6}
  ]
 }
 Trails_Condensed {
  emitterSpawnScript [&FN_RibbonIDOffsetUtility()]
  emitterUpdateScript [
   &EmitterState(Life Cycle Mode=::NewEnumerator1, Loop Duration=13.267, UseLoopDelay=false)
   &FN_SpawnAnimTrail(
    Number of Anim Trails=2
    Spawn Mode:ENiagaraSkeletalSpawning=::NewEnumerator0
    Switch on ENiagara_AnimTrail_Source:ENiagara_AnimTrail_Source=::NewEnumerator1
   )
  ]
  particleSpawnScript [
   &PartitionParticles()
   &FN_InitializeAnimTrailBoneIndecies()
   &FN_InitializeAnimTrailBoneIndecies(AnimTrail Selection=1, First Bone or Socket Index=2, Second Bone or Socket Index=3) {module=FN_InitializeAnimTrailBoneIndecies001}
   &FN_InitializeAnimTrail(AnimTrailWidth=1.5, Lifetime=&Multiply_Float(A=0.35, B=$User.FXActive), SkeletalMesh=$System.AnimTrail SKM)
  ]
  particleUpdateScript [
   &InheritSourceMovement(ApplyForce=false, ApplyPosition=true, ApplyPositionFalloff=false)
   &SolveForcesAndVelocity(Clamp Velocity=false, Limit Acceleration=false)
   &FN_UpdateAnimTrail(SkeletalMesh=$System.AnimTrail SKM) {module=FN_UpdateAnimTrail001}
   &ScaleColor(
    Linear Color Curve=DataInterface() {propertyValues {
      AlphaCurve {keys [<at 0 0>,<at 0.0851 0.25>,<at 0.2213 1>,<at 0.6068 0.26>,<at 0.8479 0.1>,<at 1 0>]}
      BlueCurve {keys [<at 0 0.0095>,<at 0.0701 0.0676>,<at 0.1641 0.0676>,<at 0.299 0>,<at 0.4665 0>]}
      GreenCurve {keys [<at 0 0.3923>,<at 0.0701 0.6654>,<at 0.1641 0.6654>,<at 0.299 0.267>,<at 0.4665 0.1274>]}
      RedCurve {keys [<at 0 1>,<at 0.0701 1>,<at 0.1641 1>,<at 0.299 1>,<at 0.4665 0.651>]}
     }}
    Scale Mode=2
   ) {module=ScaleColor001}
   &ScaleColor(
    Scale Alpha=&Multiply_Float(
     A=0.5
     B=&FloatFromCurve(
      FloatCurve=DataInterface() {propertyValues {Curve {keys [
          {interpMode=RCIM_Cubic, leaveTangent=0.0323, tangentMode=RCTM_Break, time=0, value=0}
          {arriveTangent=-1.8985, interpMode=RCIM_Cubic, leaveTangent=-1.8985, tangentMode=RCTM_User, time=0.1, value=1}
          {interpMode=RCIM_Cubic, tangentMode=RCTM_Break, time=1, value=0}
         ]}}}
      Scale Curve=&Multiply_Float(A=&FN_FrontEndFloatModifier(Gameplay=0.3, Front End=0.2), B=$User.FXActive)
     )
    )
    Scale RGB=&Multiply_Vector(
     A=<1 1 1>
     B=&VectorFromFloat(Value=&FloatFromCurve(FloatCurve=DataInterface() {propertyValues {Curve {keys [{arriveTangent=-0.005, interpMode=RCIM_Cubic, leaveTangent=-0.005, tangentMode=RCTM_User, time=0, value=1}, <at 1 0>]}}}))
    )
    ScaleA=true
    ScaleRGB=true
   )
   &ScaleRibbonWidth(Ribbon Width Scale=&FloatFromCurve(
     FloatCurve=DataInterface() {propertyValues {Curve {keys [<at 0 1 RCIM_Cubic>, {arriveTangent=-1, interpMode=RCIM_Cubic, leaveTangent=-1, time=0.3727, value=0.4975}, <at 1 0 RCIM_Cubic>]}}}
     Scale Curve=$User.FXActive
    ))
   /SetVariables(Particles.Color_light:LinearColor=&MakeLinearColorFromVectorAndFloat(
     "Vector (RGB)"=&Multiply_VectorByFloat(
      Vector=/Niagara/DynamicInputs/Math/LerpMultipleVectors(
       Alpha:NiagaraFloat=&SelectFloatFromArray(Direct Array Index=$Particles.PartitionParticles.Partition, Float Selection Array=DataInterface() {propertyValues {FloatData [0,1]}})
       V 1:Vector3f=&MakeVectorFromLinearColorRGB(Color=&ColorFromCurve(ColorCurve=DataInterface() {propertyValues {
           AlphaCurve {keys []}
           BlueCurve {keys [<at 0.0416 0.7301>,<at 0.1852 0.1084 RCIM_Cubic>]}
           GreenCurve {keys [<at 0.0416 2.4528>,<at 0.1852 1.16 RCIM_Cubic>]}
           RedCurve {keys [<at 0.0416 3>,<at 0.1852 2 RCIM_Cubic>]}
          }}))
       V 2:Vector3f=&MakeVectorFromLinearColorRGB(Color=&ColorFromCurve(ColorCurve=DataInterface() {propertyValues {
           AlphaCurve {keys []}
           BlueCurve {keys [<at 0.0416 0.7301>,<at 0.1852 0.1084 RCIM_Cubic>]}
           GreenCurve {keys [<at 0.0416 2.4528>,<at 0.1852 1.16 RCIM_Cubic>]}
           RedCurve {keys [<at 0.0416 3>,<at 0.1852 2 RCIM_Cubic>]}
          }}))
       V 3:Vector3f=?
       V 4:Vector3f=?
       V 5:Vector3f=?
       Vector Count:ENiagara_MultipleLerpCount=::NewEnumerator3
      )
      Float=1
     )
     "Float (Alpha)"=&FloatFromCurve(
      FloatCurve=DataInterface() {propertyValues {Curve {keys [
          {interpMode=RCIM_Cubic, leaveTangent=0.0323, tangentMode=RCTM_Break, time=0, value=0}
          {arriveTangent=-1.8985, interpMode=RCIM_Cubic, leaveTangent=-1.8985, tangentMode=RCTM_User, time=0.1, value=1}
          {interpMode=RCIM_Cubic, tangentMode=RCTM_Break, time=1, value=0}
         ]}}}
      Scale Curve=&Multiply_Float(A=&FN_FrontEndFloatModifier(Gameplay=0.15, Front End=0.1), B=$User.FXActive)
     )
    )) {isSetParameters=true}
   /SetVariables(Particles.Color_light_A:LinearColor=&MakeLinearColorFromVectorAndFloat(
     "Vector (RGB)"=&Multiply_VectorByFloat(
      Vector=&LerpMultipleVectors(
       Alpha:NiagaraFloat=&SelectFloatFromArray(Direct Array Index=$Particles.PartitionParticles.Partition, Float Selection Array=DataInterface() {propertyValues {FloatData [0,1]}})
       V 1:Vector3f=&MakeVectorFromLinearColorRGB(Color=&ColorFromCurve(ColorCurve=DataInterface() {propertyValues {
           AlphaCurve {keys []}
           BlueCurve {keys [<at 0.0416 0.7301>,<at 0.4586 0.2982 RCIM_Cubic>]}
           GreenCurve {keys [<at 0.0416 2.4528>,<at 0.4586 1.1371 RCIM_Cubic>]}
           RedCurve {keys [<at 0.0416 3>,<at 0.4586 2 RCIM_Cubic>]}
          }}))
       V 2:Vector3f=&MakeVectorFromLinearColorRGB(Color=&ColorFromCurve(ColorCurve=DataInterface() {propertyValues {
           AlphaCurve {keys []}
           BlueCurve {keys [<at 0.0416 0.7301>,<at 0.4586 0.2982 RCIM_Cubic>]}
           GreenCurve {keys [<at 0.0416 2.4528>,<at 0.4586 1.1371 RCIM_Cubic>]}
           RedCurve {keys [<at 0.0416 3>,<at 0.4586 2 RCIM_Cubic>]}
          }}))
       V 3:Vector3f=?
       V 4:Vector3f=?
       V 5:Vector3f=?
       Vector Count:ENiagara_MultipleLerpCount=::NewEnumerator3
      )
      Float=1
     )
     "Float (Alpha)"=&FloatFromCurve(
      FloatCurve=DataInterface() {propertyValues {Curve {keys [
          {interpMode=RCIM_Cubic, leaveTangent=0.0323, tangentMode=RCTM_Break, time=0, value=0}
          {arriveTangent=-1.8985, interpMode=RCIM_Cubic, leaveTangent=-1.8985, tangentMode=RCTM_User, time=0.1, value=1}
          {interpMode=RCIM_Cubic, tangentMode=RCTM_Break, time=1, value=0}
         ]}}}
      Scale Curve=&Multiply_Float(A=0.01, B=$User.FXActive)
     )
    )) {isSetParameters=true}
   /SetVariables(Particles.Ribbon Width light:NiagaraFloat=&Multiply_Float(A=15, B=&FloatFromCurve(FloatCurve=DataInterface() {propertyValues {Curve {keys [<at 0 1>,<at 1 0>]}}}))) {isSetParameters=true}
   &DynamicMaterialParameters(
    Index 0 Param 2=3
    Module.Index0Float0Write=true
    Module.Index0Float1Write=true
    Module.Index0Float2Write=true
    Module.Index0Float3Write=true
   )
   /SetVariables(Particles.Filler Color:LinearColor=&MultiplyLinearColor(
     Linear Color=$Particles.Color
     Scale Factor=/Niagara/DynamicInputs/TypeConversions/MakeCustomVector4FromBool(
      Bool:NiagaraBool=&Boolean_OrOperation(Bool1=$User.bIsFrontEnd, Bool2=$User.bIsFrontEndPreview)
      True Vector4:Vector4f=<1.5 1.5 1.5 1>
      False Vector4:Vector4f=<2 2 2 1>
     )
    )) {isSetParameters=true}
   /SetVariables(
    Particles.Default Color:LinearColor=&MultiplyLinearColor(Linear Color=$Particles.Color, Scale Factor=<2 2 2 1>)
    Particles.Default Width:NiagaraFloat=&Multiply_Float(
     A=10
     B=&FloatFromCurve(
      FloatCurve=DataInterface() {propertyValues {Curve {keys [<at 0 1 RCIM_Cubic>, {arriveTangent=-1, interpMode=RCIM_Cubic, leaveTangent=-1, time=0.3727, value=0.4975}, <at 1 0 RCIM_Cubic>]}}}
      Scale Curve=&Multiply_Float(A=0.4, B=$User.FXActive)
     )
    )
   ) {isSetParameters=true}
  ]
  rendererClasses [&NiagaraRibbonRendererProperties]
  renderers [
   {rendererClass=&NiagaraRibbonRendererProperties}
   {rendererClass=&NiagaraRibbonRendererProperties, rendererIndex=1}
   {rendererClass=&NiagaraRibbonRendererProperties, rendererIndex=2}
   {rendererClass=&NiagaraRibbonRendererProperties, rendererIndex=3}
  ]
 }
}
summary {
 systemName=NS_Emote_SodaPop
 emitters [
  {
   emitterName=Spotlights
   rendererClasses [
    &NiagaraMeshRendererProperties
    &NiagaraSpriteRendererProperties
    &NiagaraDecalRendererProperties
    &NiagaraLightRendererProperties
   ]
  }
  {emitterName=Sparkles, rendererClasses [&NiagaraSpriteRendererProperties]}
  {emitterName=PixieDust, rendererClasses [&NiagaraSpriteRendererProperties]}
  {emitterName=GroundSparkles, rendererClasses [&NiagaraSpriteRendererProperties]}
  {emitterName=Trails_Condensed, rendererClasses [&NiagaraRibbonRendererProperties]}
  {emitterName=BodyGlow, rendererClasses [&NiagaraMeshRendererProperties]}
 ]
 userVariables [
  User.bIsFrontEnd {type=NiagaraBool}
  User.bIsFrontEndPreview {type=NiagaraBool}
  User.FXActive {type=NiagaraFloat, defaultValue=0, description=""}
  User.OwnerLocalVelocity {type=Vector3f, description="", defaultValue {struct=Vector3f, value=<0 0 0>}}
  User.RibbonIDOffset {type=NiagaraInt32}
 ]
}
