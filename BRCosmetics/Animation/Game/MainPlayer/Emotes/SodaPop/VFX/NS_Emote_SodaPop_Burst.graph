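# Niagara burst effect for the SodaPop player emote, built from a single MeshSwirls mesh emitter that spawns a short, curve-tapered stream of swirl meshes.
# Each swirl starts raised above the origin, eases down to a lower point, grows and spins progressively faster over its sub-second life.
# Curves drive a dynamic material parameter and a warm orange-yellow color, with a faint alpha fade scaled by Fortnite time of day.
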NiagaraSystem /BRCosmetics/Animation/Game/MainPlayer/Emotes/SodaPop/VFX/NS_Emote_SodaPop_Burst
dependencies {
 usedDataInterfaces [&NiagaraDataInterfaceCurve, &NiagaraDataInterfaceColorCurve, &NiagaraDataInterfaceSkeletalMesh]
 usedDynamicInputs []
 usedModules [
  &SystemState
  &self:SystemSpawnScript.SystemScriptSource.SystemScriptGraph.NiagaraNodeAssignment_0.SetVariables_#1
 ]
}
emitters {MeshSwirls {
  emitterSpawnScript [/SetVariables(Emitter.DarkColor:LinearColor=<rgba 0.6458 0.515 0.4171 1>, Emitter.MainColor:LinearColor=<rgba 1 0.7334 0.5156 1>) {isSetParameters=true}]
  emitterUpdateScript [
   &EmitterState(Life Cycle Mode=::NewEnumerator1, Loop Behavior=::NewEnumerator1, Loop Duration=3, UseLoopDelay=false)
   &SpawnBurst_Instantaneous(Spawn Time=10.9233, Use Spawn Probability=false, UseLoopCountLimit=false)
   &SpawnRate(
    SpawnRate=&Multiply_Float(
     A=5
     B=&FloatFromCurve(FloatCurve=DataInterface() {propertyValues {Curve {keys [<at 0.5415 1.2973>,<at 0.7 0>]}}}, CurveIndex=$Emitter.Age)
    )
    Use Spawn Probability=false
   ) {module=SpawnRate001}
  ]
  particleSpawnScript [
   &InitializeParticle(
    Color Mode=::NewEnumerator0
    Lifetime Max=0.9
    Lifetime Min=0.7
    Lifetime Mode=::NewEnumerator1
    Mesh Scale=&Multiply_VectorByFloat(Vector=<1.55 1.55 1.2>, Float=&RandomRangeFloat(Minimum=1.5, Maximum=1.2))
    Mesh Scale Mode=::NewEnumerator4
    Position Offset=<0 0 130>
    UsePositionOffset=true
    Write Material Random=true
   )
   &InitialMeshOrientation(Mesh Orientation Mode=::NewEnumerator1, Rotation=<-0.5 0 1>, Use Rotation Vector=true) {module=InitialMeshOrientation001}
   &ShapeLocation(
    Cylinder Height=15
    Cylinder Height Midpoint=0
    Cylinder Radius=0
    MaskSpawnGroup=false
    Override Seed=false
    Shape Primitive=::NewEnumerator1
    Surface Only=false
    UseOwnerScale=true
   )
  ]
  particleUpdateScript [
   &ParticleState()
   /SetVariables(Particles.Position=&Lerp_Vector(
     A=Linked() {linkedVariable {name=Particles.Initial.Position, type=NiagaraPosition}}
     B=<0 0 40>
     Alpha=&FloatFromCurve(
      FloatCurve=DataInterface() {propertyValues {Curve {
         defaultValue=1
         keys [{interpMode=RCIM_Cubic, tangentMode=RCTM_User, time=0.1384, value=0}, {interpMode=RCIM_Cubic, tangentMode=RCTM_User, time=1, value=1}]
        }}}
      Scale Curve=0.9
     )
    )) {isSetParameters=true}
   &DynamicMaterialParameters(
    Index 0 Param 1=&FloatFromCurve(FloatCurve=DataInterface() {propertyValues {Curve {keys [{interpMode=RCIM_Cubic, tangentMode=RCTM_User, time=0.5, value=0}, {interpMode=RCIM_Cubic, tangentMode=RCTM_User, time=1, value=1}]}}})
    Index 0 Param 3=1
    Module.Index0Float0Write=true
    Module.Index0Float1Write=true
    Module.Index0Float2Write=true
    Module.Index0Float3Write=true
   )
   &ScaleMeshSize(Scale Factor=&Multiply_VectorByFloat(
     Vector=<1 1 1.5>
     Float=&FloatFromCurve(FloatCurve=DataInterface() {propertyValues {Curve {keys [
          {arriveTangent=0.034, interpMode=RCIM_Cubic, leaveTangent=0.034, tangentMode=RCTM_User, time=0, value=0.35}
          {arriveTangent=0.5178, interpMode=RCIM_Cubic, leaveTangent=0.5178, tangentMode=RCTM_User, time=0.6033, value=0.5613}
          {arriveTangent=1.2952, interpMode=RCIM_Cubic, leaveTangent=1.2952, tangentMode=RCTM_User, time=0.8691, value=0.75}
          <at 1 1 RCIM_Cubic>
         ]}}})
    ))
   &UpdateMeshOrientation(
    Rotation Rate=&FloatFromCurve(FloatCurve=DataInterface() {propertyValues {Curve {keys [<at 0 0>,<at 1 1>]}}}, Scale Curve=2)
    Rotation Vector=<0 0 1>
   )
   &ScaleColor(
    Scale Alpha=&FloatFromCurve(
     FloatCurve=DataInterface() {propertyValues {Curve {keys [
         {arriveTangent=0.2136, interpMode=RCIM_Cubic, leaveTangent=0.2136, tangentMode=RCTM_User, time=0, value=0}
         {arriveTangent=0.0562, interpMode=RCIM_Cubic, leaveTangent=0.0562, tangentMode=RCTM_User, time=0.3118, value=0.5992}
         {arriveTangent=-0.4718, interpMode=RCIM_Cubic, leaveTangent=-0.4718, tangentMode=RCTM_User, time=0.6074, value=0.2937}
         {arriveTangent=0.0714, interpMode=RCIM_Cubic, leaveTangent=0.0714, tangentMode=RCTM_User, time=1, value=0}
        ]}}}
     Scale Curve=&Multiply_Float(A=0.25, B=&FNTimeOfDayScalar(Night Scale=0.6, Morning Scale=0.9, Evening Scale=0.8))
    )
    Scale RGB=&MakeVectorFromLinearColorRGB(Color=&ColorFromCurve(ColorCurve=DataInterface() {propertyValues {
        AlphaCurve {keys []}
        BlueCurve {keys [
          <at 0.0198 0>
          {arriveTangent=0.4781, interpMode=RCIM_Cubic, leaveTangent=0.4781, time=0.1345, value=0}
          <at 0.3301 0.1483>
          <at 0.6167 0>
          <at 0.9005 0 RCIM_Cubic>
         ]}
        GreenCurve {keys [
          <at 0.0198 0.489>
          {arriveTangent=2.3183, interpMode=RCIM_Cubic, leaveTangent=2.3183, time=0.1345, value=0.9368}
          <at 0.3301 1.2082>
          <at 0.6167 0.9302>
          <at 0.9005 0.1746 RCIM_Cubic>
         ]}
        RedCurve {keys [
          <at 0.0198 1>
          {arriveTangent=3.5535, interpMode=RCIM_Cubic, leaveTangent=3.5535, time=0.1345, value=2}
          <at 0.3301 2.1023>
          <at 0.6167 2>
          <at 0.9005 1 RCIM_Cubic>
         ]}
       }}))
    ScaleA=true
    ScaleRGB=true
   )
  ]
  rendererClasses [&NiagaraMeshRendererProperties]
  renderers [{rendererClass=&NiagaraMeshRendererProperties, rendererIndex=0}]
 }}
summary {
 systemName=NS_Emote_SodaPop_Burst
 emitters [{emitterName=MeshSwirls, rendererClasses [&NiagaraMeshRendererProperties]}]
 userVariables []
}
