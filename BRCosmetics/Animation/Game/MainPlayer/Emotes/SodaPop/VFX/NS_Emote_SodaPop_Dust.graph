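# Niagara effect for the SodaPop emote: a burst of golden pixie dust with two flares attached to skeletal mesh bones or sockets.
# PixieDust bursts two spawn groups, masked into different sizes and lifetimes, outward from sampled points with drag, gravity and speed stretching.
# Flare follows its sampled points each update, shrinking and fading with a glow.
# Owner velocity gates inherited dust motion; time of day, platform and front-end settings adjust brightness or camera offset.
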NiagaraSystem /BRCosmetics/Animation/Game/MainPlayer/Emotes/SodaPop/VFX/NS_Emote_SodaPop_Dust
dependencies {
 usedDataInterfaces [
  &NiagaraDataInterfaceCurve
  &NiagaraDataInterfaceArrayFloat3
  &NiagaraDataInterfaceColorCurve
  &NiagaraDataInterfaceVectorCurve
  &NiagaraDataInterfaceArrayColor
  &NiagaraDataInterfacePlatformSet
  &NiagaraDataInterfaceSkeletalMesh
 ]
 usedDynamicInputs [&RandomRangeVector]
 usedModules [
  &SystemState
  &self:SystemSpawnScript.SystemScriptSource.SystemScriptGraph.NiagaraNodeAssignment_0.SetVariables_#1
  &self:SystemSpawnScript.SystemScriptSource.SystemScriptGraph.NiagaraNodeAssignment_1.SetVariables_#2
 ]
}
emitters {
 Flare {
  emitterUpdateScript [
   &EmitterState()
   &SpawnBurst_Instantaneous(Spawn Count=2, Use Spawn Probability=false, UseLoopCountLimit=false)
  ]
  particleSpawnScript [
   &PartitionParticles()
   &InitializeParticle(
    Color=&SelectLinearColorFromArray(
     Array Sampling Mode=::NewEnumerator1
     Direct Array Index=$Particles.PartitionParticles.Partition
     Color Selection Array=DataInterface() {propertyValues {ColorData [<rgba 1 0.6759 0.1098 1>, <rgba 1 0.6759 0.1098 1>]}}
     Random Seed=?
    )
    Sprite Rotation Mode=::NewEnumerator0
    Sprite Size=<30 50>
    Sprite Size Mode=::NewEnumerator4
    UsePositionOffset=false
    Write Material Random=true
   )
   &DynamicMaterialParameters(
    Index 0 Param 1=&RandomRangeFloat(Minimum=0.4, Maximum=0.6)
    Index 0 Param 2=1
    Index 0 Param 3=1
    Index 0 Param 4=1
    Module.Param0WriteEnabled=true
    Module.Param1WriteEnabled=true
    Module.Param2WriteEnabled=true
    Module.Param3WriteEnabled=true
   )
   /SetVariables(Particles.Dark_Glow_CameraOffset=&MakeCustomFloatFromBool(Bool=$User.bIsFrontEndPreview, True Float=1, False Float=50)) {isSetParameters=true}
  ]
  particleUpdateScript [
   &ParticleState()
   &ScaleSpriteSize(
    Uniform Curve Scale=5
    Uniform Curve Sprite Scale=DataInterface() {propertyValues {Curve {
       defaultValue=0
       keys [
        {arriveTangent=-2.5541, interpMode=RCIM_Cubic, leaveTangent=-2.5541, tangentMode=RCTM_User, time=0.0319, value=1.2933}
        {arriveTangent=-2.5606, interpMode=RCIM_Cubic, leaveTangent=-2.5606, tangentMode=RCTM_User, time=0.1012, value=0.7981}
        {interpMode=RCIM_Cubic, tangentMode=RCTM_User, time=1, value=0}
       ]
      }}}
   ) {enabled=false}
   &ScaleSpriteSize(
    Uniform Curve Scale=5
    Uniform Curve Sprite Scale=DataInterface() {propertyValues {Curve {
       defaultValue=0
       keys [
        {arriveTangent=-2.5541, interpMode=RCIM_Cubic, leaveTangent=-2.5541, tangentMode=RCTM_User, time=0.0381, value=1.2781}
        {arriveTangent=-2.1465, interpMode=RCIM_Cubic, leaveTangent=-2.1465, tangentMode=RCTM_User, time=0.152, value=0.7981}
        {interpMode=RCIM_Cubic, tangentMode=RCTM_User, time=1, value=0}
       ]
      }}}
   ) {module=ScaleSpriteSize001}
   &SkeletalMeshLocation(
    Bone / Socket Index=$Particles.PartitionParticles.Partition
    Bone Sampling Mode=::NewEnumerator4
    MaskSpawnGroup=false
    Sampled Position Offset=&SelectVectorFromArray(
     Direct Array Index=$Particles.PartitionParticles.Partition
     Vector Selection Array=DataInterface() {propertyValues {FloatData [<0 0 -10>, <0 0 -10>]}}
    )
    Skeletal Mesh=$System.Skeletal Mesh Reader
   )
   &ScaleColor(
    Scale Alpha=&Multiply_Float(
     A=1
     B=&FloatFromCurve(
      FloatCurve=DataInterface() {propertyValues {Curve {
         defaultValue=0
         keys [
          {interpMode=RCIM_Cubic, tangentMode=RCTM_User, time=0, value=1}
          {arriveTangent=-1, interpMode=RCIM_Cubic, leaveTangent=-1, time=0.2827, value=0.3229}
          {interpMode=RCIM_Cubic, tangentMode=RCTM_User, time=1, value=0}
         ]
        }}}
      Scale Curve=&Multiply_Float(
       A=&Multiply_Float(A=0.1, B=&FN_PickFloatDependOnPlatform(TargetPlatformActive=1, TargetPlatformNoActive=0.5))
       B=&FNTimeOfDayScalar(Night Scale=0.6, Morning Scale=0.6, Evening Scale=0.6)
      )
     )
    )
    Scale RGB=<1.5 1.5 1.5>
    ScaleA=true
    ScaleRGB=true
   )
   /SetVariables(Particles.Dark_Glow=&MakeLinearColorFromVectorAndFloat(
     "Vector (RGB)"=&MakeVectorFromLinearColorRGB(Color=<rgba 1 0.3964 0 0.2087>)
     "Float (Alpha)"=&Multiply_Float(
      A=&Multiply_Float(A=&FNTimeOfDayScalar(Night Scale=0.5, Morning Scale=0.3, Evening Scale=0.4), B=1.5)
      B=&FloatFromCurve(
       FloatCurve=DataInterface() {propertyValues {Curve {keys [
           <at 0 0>
           {arriveTangent=1.1111, leaveTangent=1.1111, time=0.1, value=1}
           {arriveTangent=-1.1111, leaveTangent=-1.1111, time=0.9, value=1}
           <at 1 0>
          ]}}}
       Scale Curve=0.03
      )
     )
    )) {isSetParameters=true}
   /SetVariables(Particles.Dark_Glow_Size=&Multiply_Vector2(A=$Particles.SpriteSize, B=&Vector2DFromFloat(Value=1.5))) {isSetParameters=true}
   &FN_Light_Attributes(
    Color=&ColorFromCurve(ColorCurve=DataInterface() {propertyValues {AlphaCurve {keys []}, BlueCurve {keys [<at 0.1896 1>]}, GreenCurve {keys [<at 0.1896 0.3504>]}, RedCurve {keys [<at 0.1896 0.0798>]}}})
    Create Color Attribute=true
    Create Exponent Attribute=true
    Create Position Attribute=true
    Create Radius Attribute=true
    Exponent=2
    Position=<0 -2 0>
    Radius=&FloatFromCurve(FloatCurve=DataInterface() {propertyValues {Curve {keys [<at 0 1>,<at 1 1>]}}}, Scale Curve=50)
    Scale Color=&FloatFromCurve(
     FloatCurve=DataInterface() {propertyValues {Curve {keys [
         {arriveTangent=-0.0015, interpMode=RCIM_Cubic, leaveTangent=-0.0015, tangentMode=RCTM_User, time=0, value=0}
         {arriveTangent=-0.0105, interpMode=RCIM_Cubic, leaveTangent=-0.0105, tangentMode=RCTM_User, time=0.5, value=1}
         {arriveTangent=-0.9284, interpMode=RCIM_Cubic, leaveTangent=-0.9284, tangentMode=RCTM_User, time=1, value=0}
        ]}}}
     Scale Curve=&Multiply_Float(A=&FNTimeOfDayScalar(Night Scale=0.5, Morning Scale=0.6, Evening Scale=0.6), B=15)
    )
   ) {enabled=false}
  ]
  rendererClasses [&NiagaraSpriteRendererProperties]
  renderers [
   {rendererClass=&NiagaraSpriteRendererProperties}
   {rendererClass=&NiagaraSpriteRendererProperties, rendererIndex=1}
  ]
 }
 PixieDust {
  emitterUpdateScript [
   &EmitterState()
   &SpawnBurst_Instantaneous(Spawn Count=22, Use Spawn Probability=false, UseLoopCountLimit=false)
   &SpawnRate(
    SpawnRate=&FloatFromCurve(
     FloatCurve=DataInterface() {propertyValues {Curve {keys [<at 0.016 1>,<at 0.0296 -0.0235>]}}}
     CurveIndex=$Emitter.NormalizedLoopAge
     Scale Curve=100
    )
    Use Spawn Probability=false
   ) {enabled=false}
   &SpawnBurst_Instantaneous(Spawn Count=16, Spawn Group=1, Use Spawn Probability=false, UseLoopCountLimit=false) {module=SpawnBurst_Instantaneous001}
   &SpawnRate(
    SpawnGroup=1
    SpawnRate=&FloatFromCurve(
     FloatCurve=DataInterface() {propertyValues {Curve {keys [<at 0.016 1>,<at 0.0296 -0.0235>]}}}
     CurveIndex=$Emitter.NormalizedLoopAge
     Scale Curve=100
    )
    Use Spawn Probability=false
   ) {enabled=false, module=SpawnRate002}
  ]
  particleSpawnScript [
   /SetVariables(Particles.VisibilityTag=&SpawnGroup()) {isSetParameters=true}
   &PartitionParticles()
   &InitializeParticle(
    Color Maximum=<rgba 1 0.5602 0.3258 1.1>
    Color Minimum=<rgba 1 0.8063 0.3229 1.1>
    Color Mode=::NewEnumerator2
    Lifetime=&MaskFloat(
     Mask Source=0
     Mask Parameter=?
     Mask=0
     Masked Float=&RandomRangeFloat(Minimum=0.2, Maximum=0.6)
     Passthrough Float=&RandomRangeFloat(Minimum=0.3, Maximum=1)
    )
    Position Offset=<2 0 -2>
    Sprite Rotation Mode=::NewEnumerator2
    Sprite Size Mode=::NewEnumerator3
    Uniform Sprite Size=&MaskFloat(
     Mask Source=0
     Mask Parameter=?
     Mask=0
     Masked Float=&RandomRangeFloat(Minimum=35, Maximum=65)
     Passthrough Float=&RandomRangeFloat(Minimum=2, Maximum=5)
    )
    UsePositionOffset=true
    Write Material Random=true
   )
   &SkeletalMeshLocation(
    Bone / Socket Index=$Particles.PartitionParticles.Partition
    Bone Sampling Mode=::NewEnumerator4
    MaskSpawnGroup=false
    Sampled Position Offset=&SelectVectorFromArray(
     Direct Array Index=$Particles.PartitionParticles.Partition
     Vector Selection Array=DataInterface() {propertyValues {FloatData [<-5 0 0>, <5 0 0>]}}
    )
    Skeletal Mesh=$System.Skeletal Mesh
   ) {module=SkeletalMeshLocation001}
   &ShapeLocation(
    Additional Quaternion Rotation=false
    MaskSpawnGroup=false
    Override Seed=false
    Rotation Mode=::NewEnumerator3
    Rotation Quaternion=$Particles.SkeletalMeshLocation001.SampledRotation
    Sphere Radius=2
    UseOwnerScale=true
   ) {enabled=false}
   &AddVelocity(
    Module.Use Falloff Distance=false
    Override Seed=false
    Velocity Mode=::NewEnumerator1
    Velocity Origin=$Output.SkeletalMeshLocation001.SampledPositionWithoutOffset
    Velocity Speed=&RandomRangeFloat(Minimum=50, Maximum=200)
   )
   &AddVelocity(
    Additional Quaternion Rotation=false
    Cone Angle=70
    Module.UseConeFalloff=false
    Module.UseVelDistribution=false
    Override Seed=false
    Rotation Mode=::NewEnumerator3
    Rotation Quaternion:Quat4f=$Particles.SkeletalMeshLocation001.SampledRotation
    Velocity Mode=::NewEnumerator2
    Velocity Speed=100
   ) {module=AddVelocity001}
  ]
  particleUpdateScript [
   &ParticleState()
   &ScaleSpriteSize(
    Scale Sprite Size Mode=::NewEnumerator0
    Uniform Scale Factor=&MaskFloat(
     Mask Source=0
     Mask Parameter=?
     Mask=0
     Masked Float=&FloatFromCurve(FloatCurve=DataInterface() {propertyValues {Curve {
         defaultValue=0.6
         keys [{interpMode=RCIM_Cubic, tangentMode=RCTM_User, time=0, value=0}, {interpMode=RCIM_Cubic, tangentMode=RCTM_User, time=1, value=1}]
        }}})
     Passthrough Float=&FloatFromCurve(FloatCurve=DataInterface() {propertyValues {Curve {keys [
          <at 0 0 RCIM_Cubic>
          {arriveTangent=0.5233, interpMode=RCIM_Cubic, leaveTangent=0.5233, time=0.2, value=0.95}
          {arriveTangent=-0.3966, interpMode=RCIM_Cubic, leaveTangent=-0.3966, time=0.4754, value=0.2488}
          {arriveTangent=-0.4742, interpMode=RCIM_Cubic, leaveTangent=-0.4742, time=0.7179, value=0.7446}
          <at 1 0 RCIM_Cubic>
         ]}}})
    )
   )
   &InheritSourceMovement(
    Applied Position Scale=&MakeCustomVectorFromBool(
     Bool=&SetBoolByFloatComparison(A=&VectorLength(Vector=$User.OwnerLocalVelocity), B=100)
     True Vector=<0 0 0>
     False Vector=<0.7 0.7 0.7>
    )
    ApplyForce=false
    ApplyPosition=true
    ApplyPositionFalloff=false
   )
   &Drag(Drag=&RandomRangeFloat(Minimum=0.75, Maximum=2), Use Linear Drag=true, Use Rotational Drag=true)
   &InheritVelocity(
    Inherited Velocity Amount Scale=&MakeCustomVectorFromBool(
     Bool=&SetBoolByFloatComparison(A=&VectorLength(Vector=$User.OwnerLocalVelocity), B=100)
     True Vector=<0 0 0>
     False Vector=<1 1 -1>
    )
    Inherited Velocity Speed Limit=4
    Speed Limit=true
    UseSpeedThreshold=false
    Velocity Source=$Particles.SkeletalMeshLocation001.SampledVelocity
   )
   &GravityForce(Gravity=<0 0 -130>)
   &SolveForcesAndVelocity(Clamp Velocity=false, Limit Acceleration=false)
   &ScaleSpriteSizeBySpeed(
    Max Scale Factor=/Niagara/DynamicInputs/Helpers/MaskVector2d(
     Mask Source:NiagaraInt32=0
     Mask Parameter:NiagaraInt32=?
     Mask:NiagaraInt32=0
     Masked Vector2D:Vector2f=<1 1>
     Passthrough Vector2D:Vector2f=<2 20>
    )
    Min Scale Factor=&MaskVector2d(
     Mask Source:NiagaraInt32=0
     Mask Parameter:NiagaraInt32=?
     Mask:NiagaraInt32=0
     Masked Vector2D:Vector2f=<1 1>
     Passthrough Vector2D:Vector2f=<0.75 1>
    )
   )
   &ScaleColor(
    Scale Alpha=&FloatFromCurve(
     FloatCurve=DataInterface() {propertyValues {Curve {
        defaultValue=0.006
        keys [
         {interpMode=RCIM_Cubic, tangentMode=RCTM_User, time=0, value=0}
         {arriveTangent=0.5346, interpMode=RCIM_Cubic, leaveTangent=0.5346, time=0.4211, value=1}
         {arriveTangent=-1.7143, interpMode=RCIM_Cubic, leaveTangent=-1.7143, tangentMode=RCTM_User, time=0.7084, value=0.3787}
         {tangentMode=RCTM_User, time=1, value=0}
        ]
       }}}
     Scale Curve=&Multiply_Float(
      A=&FNTimeOfDayScalar(Night Scale=0.6, Morning Scale=0.8, Evening Scale=0.8)
      B=&FN_FrontEndFloatModifier(
       Gameplay=&RandomRangeFloat(Minimum=0.25, Maximum=0.35)
       Front End=&RandomRangeFloat(Minimum=0.2, Maximum=0.3)
      )
     )
    )
    Scale RGB=&Multiply_Vector(
     A=&MakeVectorFromLinearColorRGB(Color=&ColorFromCurve(
       ColorCurve=DataInterface() {propertyValues {
         AlphaCurve {keys []}
         BlueCurve {keys [<at 0.0038 7>, <at 0.0437 3>, <at 0.0954 2>, {interpMode=RCIM_Cubic, leaveTangent=-14.4932, time=0.1442, value=1}, <at 0.1519 1 RCIM_Cubic>]}
         GreenCurve {keys [<at 0.0038 7>, <at 0.0437 3>, <at 0.0954 2>, {interpMode=RCIM_Cubic, leaveTangent=-14.4932, time=0.1442, value=1}, <at 0.1519 1 RCIM_Cubic>]}
         RedCurve {keys [<at 0.0038 7>, <at 0.0437 3>, <at 0.0954 2>, {interpMode=RCIM_Cubic, leaveTangent=-14.4932, time=0.1442, value=1}, <at 0.1519 1 RCIM_Cubic>]}
        }}
       CurveIndex=$Emitter.NormalizedLoopAge
      ))
     B=&VectorFromCurve(VectorCurve=DataInterface() {propertyValues {
        XCurve {keys [<at 0 1 RCIM_Cubic>,<at 1 1 RCIM_Cubic>]}
        YCurve {keys [<at 0 1 RCIM_Cubic>,<at 1 0.75 RCIM_Cubic>]}
        ZCurve {keys [<at 0 1 RCIM_Cubic>,<at 1 0.75 RCIM_Cubic>]}
       }})
    )
    ScaleA=true
    ScaleRGB=true
   )
   &DynamicMaterialParameters(
    Index 0 Param 1=&FloatFromCurve(FloatCurve=DataInterface() {propertyValues {Curve {keys [<at 0 0>,<at 1 0.02>]}}}, Scale Curve=0)
    Module.Index0Float0Write=true
    Module.Index0Float1Write=true
    Module.Index0Float2Write=true
    Module.Index0Float3Write=true
   )
   /SetVariables(
    Particles.SparkleColor:LinearColor=&MakeLinearColorFromVectorAndFloat(
     "Vector (RGB)"=&MakeVectorFromLinearColorRGB(Color=&RandomRangeLinearColor(Minimum=<rgba 1.5 0.9639 0.0054 1>, Maximum=<rgba 3 2.148 0.4033 1>))
     "Float (Alpha)"=&FloatFromCurve(FloatCurve=DataInterface() {propertyValues {Curve {
         defaultValue=1.0194
         keys [
          {interpMode=RCIM_Cubic, tangentMode=RCTM_User, time=0, value=0}
          {
           arriveTangent=7.0019
           arriveTangentWeight=0.0547
           interpMode=RCIM_Cubic
           leaveTangent=7.0019
           leaveTangentWeight=0.107
           tangentMode=RCTM_User
           time=0.1641
           value=1
          }
          {arriveTangent=0.3424, interpMode=RCIM_Cubic, leaveTangent=0.3424, time=0.3242, value=2}
          {
           arriveTangent=-6.1112
           arriveTangentWeight=0.1926
           interpMode=RCIM_Cubic
           leaveTangent=-6.1112
           leaveTangentWeight=0.1149
           tangentMode=RCTM_User
           time=0.5044
           value=1.1165
          }
          {arriveTangent=-2.2528, interpMode=RCIM_Cubic, leaveTangent=-2.2528, time=0.6763, value=0.2328}
          {tangentMode=RCTM_User, time=1, value=0}
         ]
        }}})
    )
    Particles.Vector 2D:Vector2f=&Multiply_Vector2D_ByFloat(
     Float=&FloatFromCurve(FloatCurve=DataInterface() {propertyValues {Curve {keys [
          <at 0 0 RCIM_Cubic>
          {arriveTangent=0.5233, interpMode=RCIM_Cubic, leaveTangent=0.5233, time=0.2, value=0.95}
          {arriveTangent=-0.3966, interpMode=RCIM_Cubic, leaveTangent=-0.3966, time=0.4754, value=0.2488}
          {arriveTangent=-0.4742, interpMode=RCIM_Cubic, leaveTangent=-0.4742, time=0.7179, value=0.7446}
          <at 1 0 RCIM_Cubic>
         ]}}})
     Vector2D=$Particles.Initial.SpriteSize
    )
   ) {isSetParameters=true}
  ]
  rendererClasses [&NiagaraSpriteRendererProperties]
  renderers [
   {rendererClass=&NiagaraSpriteRendererProperties}
   {rendererClass=&NiagaraSpriteRendererProperties, rendererIndex=1}
  ]
 }
}
summary {
 systemName=NS_Emote_SodaPop_Dust
 emitters [
  {emitterName=PixieDust, rendererClasses [&NiagaraSpriteRendererProperties]}
  {emitterName=Flare, rendererClasses [&NiagaraSpriteRendererProperties]}
 ]
 userVariables [
  User.bIsFrontEnd {type=NiagaraBool}
  User.bIsFrontEndPreview {type=NiagaraBool}
  User.OwnerLocalVelocity {type=Vector3f, description="", defaultValue {struct=Vector3f, value=<0 0 0>}}
 ]
}
