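# Niagara system for the SodaPop player emote that leaves lingering orange ribbon trails, named HandTrails, on the character.
# One emitter spawns continuously, using the particle partition to pick a bone or socket on the system's skeletal mesh reader plus a mirrored offset.
# Curves fade color, alpha and ribbon width over particle life while particles inherit source movement.
# Separate "light" color and width attributes are also computed, with color alpha enveloped by system age.
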NiagaraSystem /BRCosmetics/Animation/Game/MainPlayer/Emotes/SodaPop/VFX/NS_Emote_SodaPop_LingeringTrails
dependencies {
 usedDataInterfaces [
  &NiagaraDataInterfaceArrayFloat3
  &NiagaraDataInterfaceColorCurve
  &NiagaraDataInterfaceCurve
  &NiagaraDataInterfaceArrayFloat
  &NiagaraDataInterfaceSkeletalMesh
 ]
 usedDynamicInputs []
 usedModules [
  &SystemState
  &self:SystemSpawnScript.SystemScriptSource.SystemScriptGraph.NiagaraNodeAssignment_0.SetVariables_#1
 ]
}
emitters {HandTrails {
  emitterUpdateScript [
   &EmitterState()
   &SpawnRate(SpawnRate=80, Use Spawn Probability=false)
  ]
  particleSpawnScript [
   &PartitionParticles()
   &InitializeParticle(
    Color Mode=::NewEnumerator0
    Lifetime=3
    Ribbon Width=35
    Ribbon Width Mode=::NewEnumerator1
    UsePositionOffset=false
    Write Material Random=true
   )
   &SkeletalMeshLocation(
    Bone / Socket Index=$Particles.PartitionParticles.Partition
    Bone Sampling Mode=::NewEnumerator4
    MaskSpawnGroup=false
    Sampled Position Offset=&SelectVectorFromArray(
     Direct Array Index=$Particles.PartitionParticles.Partition
     Vector Selection Array=DataInterface() {propertyValues {FloatData [<-5 0 0>, <5 0 0>]}}
    )
    Skeletal Mesh=$System.Skeletal Mesh Reader
   )
  ]
  particleUpdateScript [
   &ParticleState()
   &ScaleColor(
    Linear Color Curve=DataInterface() {propertyValues {
      AlphaCurve {keys [<at 0 0>,<at 0.0851 0.25>,<at 0.2213 1>,<at 0.6068 0.26>,<at 0.8479 0.1>,<at 1 0>]}
      BlueCurve {keys [<at 0 0.0095>,<at 0.0701 0.0676>,<at 0.1641 0.0676>,<at 0.299 0>,<at 0.4665 0>]}
      GreenCurve {keys [<at 0 0.3923>,<at 0.0701 0.6654>,<at 0.1641 0.6654>,<at 0.299 0.267>,<at 0.4665 0.1274>]}
      RedCurve {keys [<at 0 1>,<at 0.0701 1>,<at 0.1641 1>,<at 0.299 1>,<at 0.4665 0.651>]}
     }}
    Scale Mode=2
   ) {module=ScaleColor001}
   &ScaleColor(
    Scale Alpha=&Multiply_Float(
     A=0.5
     B=&FloatFromCurve(
      FloatCurve=DataInterface() {propertyValues {Curve {keys [
          {interpMode=RCIM_Cubic, leaveTangent=0.0323, tangentMode=RCTM_Break, time=0, value=0}
          {arriveTangent=-1.8985, interpMode=RCIM_Cubic, leaveTangent=-1.8985, tangentMode=RCTM_User, time=0.1, value=1}
          {interpMode=RCIM_Cubic, tangentMode=RCTM_Break, time=1, value=0}
         ]}}}
      Scale Curve=0.6
     )
    )
    Scale RGB=&Multiply_Vector(
     A=<1 1 1>
     B=&VectorFromFloat(Value=&FloatFromCurve(FloatCurve=DataInterface() {propertyValues {Curve {keys [{arriveTangent=-0.005, interpMode=RCIM_Cubic, leaveTangent=-0.005, tangentMode=RCTM_User, time=0, value=1}, <at 1 0>]}}}))
    )
    ScaleA=true
    ScaleRGB=true
   )
   &ScaleRibbonWidth(Ribbon Width Scale=&FloatFromCurve(FloatCurve=DataInterface() {propertyValues {Curve {keys [<at 0 1 RCIM_Cubic>, {arriveTangent=-1, interpMode=RCIM_Cubic, leaveTangent=-1, time=0.3727, value=0.4975}, <at 1 0 RCIM_Cubic>]}}}))
   /SetVariables(Particles.Color_light:LinearColor=&MakeLinearColorFromVectorAndFloat(
     "Vector (RGB)"=&Multiply_VectorByFloat(
      Vector=/Niagara/DynamicInputs/Math/LerpMultipleVectors(
       Alpha:NiagaraFloat=&SelectFloatFromArray(Direct Array Index=$Particles.PartitionParticles.Partition, Float Selection Array=DataInterface() {propertyValues {FloatData [0,1]}})
       V 1:Vector3f=&MakeVectorFromLinearColorRGB(Color=&ColorFromCurve(ColorCurve=DataInterface() {propertyValues {
           AlphaCurve {keys []}
           BlueCurve {keys [<at 0.0416 0.3159>, {arriveTangent=-0.874, interpMode=RCIM_Cubic, leaveTangent=-0.874, time=0.1852, value=0.0767}, <at 0.4031 0>]}
           GreenCurve {keys [<at 0.0416 1.5683>, {arriveTangent=-3.4521, interpMode=RCIM_Cubic, leaveTangent=-3.4521, time=0.1852, value=0.6193}, <at 0.4031 0.3207>]}
           RedCurve {keys [<at 0.0416 3>, {arriveTangent=-2.7669, interpMode=RCIM_Cubic, leaveTangent=-2.7669, time=0.1852, value=2}, <at 0.4031 2>]}
          }}))
       V 2:Vector3f=&MakeVectorFromLinearColorRGB(Color=&ColorFromCurve(ColorCurve=DataInterface() {propertyValues {
           AlphaCurve {keys []}
           BlueCurve {keys [<at 0.0416 0.3159>, {arriveTangent=-0.874, interpMode=RCIM_Cubic, leaveTangent=-0.874, time=0.1852, value=0.0767}, <at 0.4031 0>]}
           GreenCurve {keys [<at 0.0416 1.5683>, {arriveTangent=-3.4521, interpMode=RCIM_Cubic, leaveTangent=-3.4521, time=0.1852, value=0.6193}, <at 0.4031 0.3207>]}
           RedCurve {keys [<at 0.0416 3>, {arriveTangent=-2.7669, interpMode=RCIM_Cubic, leaveTangent=-2.7669, time=0.1852, value=2}, <at 0.4031 2>]}
          }}))
       V 3:Vector3f=?
       V 4:Vector3f=?
       V 5:Vector3f=?
       Vector Count:ENiagara_MultipleLerpCount=::NewEnumerator3
      )
      Float=1
     )
     "Float (Alpha)"=&Multiply_Float(
      A=&FloatFromCurve(
       FloatCurve=DataInterface() {propertyValues {Curve {keys [
           <at -0.0083 0 RCIM_Cubic>
           {arriveTangent=4.8004, interpMode=RCIM_Cubic, leaveTangent=4.8004, time=0.1091, value=0.1519}
           {arriveTangentWeight=0.0343, interpMode=RCIM_Cubic, leaveTangentWeight=0.5666, tangentMode=RCTM_User, time=0.2, value=1}
           {arriveTangentWeight=0.5929, interpMode=RCIM_Cubic, leaveTangentWeight=0.0395, tangentMode=RCTM_User, time=1.4, value=1}
           {arriveTangent=-2.0388, interpMode=RCIM_Cubic, leaveTangent=-2.0388, time=1.628, value=0.3074}
           <at 1.9 -0.0194 RCIM_Cubic>
          ]}}}
       CurveIndex=$System.Age
      )
      B=&FloatFromCurve(
       FloatCurve=DataInterface() {propertyValues {Curve {keys [
           {interpMode=RCIM_Cubic, leaveTangent=0.0323, tangentMode=RCTM_Break, time=0, value=0}
           {arriveTangent=-1.8985, interpMode=RCIM_Cubic, leaveTangent=-1.8985, tangentMode=RCTM_User, time=0.1, value=1}
           {interpMode=RCIM_Cubic, tangentMode=RCTM_Break, time=1, value=0}
          ]}}}
       Scale Curve=0.4
      )
     )
    )) {isSetParameters=true}
   /SetVariables(Particles.Ribbon Width light:NiagaraFloat=&Multiply_Float(A=15, B=&FloatFromCurve(FloatCurve=DataInterface() {propertyValues {Curve {keys [<at 0 1>,<at 1 0>]}}}))) {isSetParameters=true}
   &DynamicMaterialParameters(
    Index 0 Param 2=3
    Module.Index0Float0Write=true
    Module.Index0Float1Write=true
    Module.Index0Float2Write=true
    Module.Index0Float3Write=true
   )
   &InheritSourceMovement(ApplyForce=false, ApplyPosition=true, ApplyPositionFalloff=false)
   &SolveForcesAndVelocity(Clamp Velocity=false, Limit Acceleration=false)
  ]
  rendererClasses [&NiagaraRibbonRendererProperties]
  renderers [{rendererClass=&NiagaraRibbonRendererProperties}]
 }}
summary {
 systemName=NS_Emote_SodaPop_LingeringTrails
 emitters [{emitterName=HandTrails, rendererClasses [&NiagaraRibbonRendererProperties]}]
 userVariables []
}
